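; GA_IronLog_Equipped is a gameplay ability for the Iron Log weapon's equipped state.
; Activation simply defers to the parent ability's behavior.
; On end, it calls the parent, then ends the stored on-damage-dealt async task if it is valid, cleaning up that listener.
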(event Ability|EventActivateAbility
  (|CallParentFunction))

(event Ability|EventOnEndAbility
  (params (bWasCancelled :type Boolean))
  (|CallParentFunction bWasCancelled)
  (bind _ondamagedealtasynctask (Variables|Default|GetOnDamageDealtAsyncTask)
    (:then
      (GameplayTasks|EndTask _ondamagedealtasynctask))
    (:else)))
