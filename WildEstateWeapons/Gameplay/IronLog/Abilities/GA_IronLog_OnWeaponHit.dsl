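; IronLog weapon-hit gameplay ability that, on a hit event, plays impact cues and applies a face screen effect to hit pawns.
; ParseData groups target-data hit results per pawn, then picks each pawn's hit nearest the average location as target data.
; It refreshes CenterScreenGE by removing and reapplying it, forced on for playtesting in place of a partial-effect choice.
; The toggleable impact cue uses the last parsed hit, not the central one.
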(event Ability|EventActivateAbilityFromEvent
  (params (EventData :type Gameplay%20Event%20Data%20Structure%20%28by%20ref%29))
  (bind (_eventtag _instigator _target _optionalobject _optionalobject2 _contexthandle _instigatortags _targettags _eventmagnitude _targetdata) (Utilities|Struct|BreakGameplayEventData EventData))
  ;[1 1 1] Activate & Get Data
  (|CallParentFunction EventData)
  (bind _outfortplayerpawn (CallFunction|GetAvatarasValidFortPlayerPawn)
    (:ValidFortPlayerPawn
      (Variables|Default|SetInstigatorASC (Ability|GetFortAbilitySystemComponent _outfortplayerpawn))
      (CallFunction|ParseData _targetdata))))

(event Custom|ParseData
  (params (TargetData :type Gameplay%20Ability%20Target%20Data%20Handle%20Structure%20%28by%20ref%29))
  (bind _currentpawn (Variables|Default|GetCurrentPawn))
  (bind _perpawnresults (Variables|Default|GetPerPawnResults))
  (bind (_value _returnvalue) (Utilities|Map|Find _perpawnresults _currentpawn))
  (bind _currenttargetdata (Variables|Default|GetCurrentTargetData))
  ;[1 1 1] Get Data
  (bind _output_get (Variables|Default|SetCurrentTargetData TargetData))
  (for (_array_element _array_index) (Ability|TargetData|GetAllActorsfromTargetData _output_get)
    (bind _asfort_pawn (Utilities|Casting|CastToFortPawn _array_element))
    (Variables|Default|SetCurrentPawn _asfort_pawn)
    (if (Ability|TargetData|TargetDataHasHitResult _currenttargetdata _array_index)
      (Variables|Default|SetCurrentHitResult (Ability|TargetData|GetHitResultfromTargetData _currenttargetdata _array_index))
      ;[1 1 1] Update map with another hitresult
      (if _returnvalue
        (bind _output_get_2 (Variables|Default|SetCurrentHitResultsWrapper (Utilities|Struct|BreakIronLogHitResultsWrapper _value)))
        (Utilities|Array|Add _output_get_2 (Variables|Default|GetCurrentHitResult))
        (Utilities|Map|Add (Variables|Default|GetPerPawnResults) (Variables|Default|GetCurrentPawn) (Utilities|Struct|MakeIronLogHitResultsWrapper _output_get_2))
        (else
          ;[1 1 1] Create a new key value pair for the map
          (Utilities|Map|Add (Variables|Default|GetPerPawnResults) (Variables|Default|GetCurrentPawn) (Utilities|Struct|MakeIronLogHitResultsWrapper (Utilities|Array|MakeArray (Variables|Default|GetCurrentHitResult))))))))
  ;[1 1 1] Get Ability Systsem Component, Check If Makeup Effect (Cue), If No Then Apply GE, If It Is, Trigger The Full Face Effect
  (bind (_keypin _valuepin) (Utilities|Map|ForEachLoop(Map) (Variables|Default|GetPerPawnResults)))
  (bind _centralhitresult (CallFunction|FindCentralHitResult _keypin (Utilities|Struct|BreakIronLogHitResultsWrapper _valuepin)))
  (Variables|Default|SetCurrentTargetData (Ability|TargetData|AbilityTargetDatafromHitResult _centralhitresult))
  (bind _output_get_3 (Variables|Default|SetTargetsAbilitySystemComponent (Ability|GetFortAbilitySystemComponent _keypin)))
  (Utilities|IsValid _output_get_3
    (:Is%20Valid
      ;; Temporarily forced true to apply the full face screen effect vs both the partial and true for playtesting. Will delete unused code or revert this change after design decision is made
      (if true
        (CallFunction|ExecuteImpactGameplayCue)
        (Utilities|FlowControl|DelayUntilNextTick)
        ;;[1 1 1] Remove Active Face Effect
        (GameplayEffects|RemoveActiveGameplayEffectbySourceEffect (Variables|Default|GetTargetsAbilitySystemComponent) (Variables|Default|GetCenterScreenGE) 0 -1)
        (Ability|ApplyGameplayEffectToTarget (Variables|Default|GetCurrentTargetData) (Variables|Default|GetCenterScreenGE))
        (else
          _)))))

(fn ExecuteImpactGameplayCue
  (bind (_bblockinghit _binitialoverlap _time _distance _location _impactpoint _normal _impactnormal _physmat _hitactor) (Collision|BreakHitResult (Variables|Default|GetCurrentHitResult)))
  (if (ScalableFloat|GetValueasBool (Variables|Default|GetHF_EnablePawnImpactEffect))
    (GameplayCue|ExecuteGameplayCueOnActor(Burst) _hitactor (Variables|Default|GetGC_Impact) (Ability|GameplayCue|MakeGameplayCueParameters :NormalizedMagnitude 0.0 :RawMagnitude 0.0 :Location _location :Normal _normal :GameplayEffectLevel 1 :AbilityLevel 1 :bReplicateLocationWhenUsingMinimalRepProxy true))))

(fn FindCentralHitResult
  (params (FortPawn :type Fort%20Pawn%20Object%20Reference) (HitResults :type Array%20of%20Hit%20Result%20Structures))
  (bind _hitresults (Variables|GetHitResults))
  (Utilities|IsValid FortPawn
    (:Is%20Valid
      (Utilities|FlowControl|Sequence
        (:then_0
          ;[1 1 1] Get the average location of all the hits
          (for _array_element _hitresults
            (LocalVariables|SetHitLocationsSum (+ (.Location (Collision|BreakHitResult _array_element)) (LocalVariables|GetHitLocationsSum))))
          (LocalVariables|SetAverageHitLocation (/ (LocalVariables|GetHitLocationsSum) (Utilities|Array|Length _hitresults))))
        (:then_1
          ;[1 1 1] Find the hit that is closest to the average
          (for _array_element_2 (Variables|GetHitResults)
            (bind _output_get (LocalVariables|SetCurrentDistance (Math|Vector|Distance(Vector) (.Location (Collision|BreakHitResult _array_element_2)) (LocalVariables|GetAverageHitLocation))))
            (if (< _output_get (LocalVariables|GetShortestDistance))
              (LocalVariables|SetMostCentralHitResult _array_element_2)
              (LocalVariables|SetShortestDistance (LocalVariables|GetCurrentDistance))))
          (return (LocalVariables|GetMostCentralHitResult)))))))
