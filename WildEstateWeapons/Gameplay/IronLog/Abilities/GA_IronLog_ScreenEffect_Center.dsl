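; Passive Iron Log ability that shows a center screen effect and an on-pawn makeup effect on a hit player, each behind a toggle.
; On activation, non-authority instances spawn the lens effect and set its Niagara duration.
; A gameplay cue is added to the owner with location and direction from the granting effect's hit result.
; On end, non-authority cleanup casts the stored lens effect to the Corners class, not Center, before destroying it.
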(event Ability|EventFailedToActivatePassiveAbility
  (params (ActorInfo :type Gameplay%20Ability%20Actor%20Info%20Structure))
  ;[1 1e-06 0.001368] Failed to activate
  (|CallParentFunction ActorInfo))

(event Ability|EventActivateAbility
  ;[1 1 1] OnActivate
  (|CallParentFunction)
  (bind _outfortplayerpawn (CallFunction|GetAvatarasValidFortPlayerPawn)
    (:ValidFortPlayerPawn
      (Utilities|FlowControl|Sequence
        (:then_0
          (Ability|HasAuthority
            (:False
              ;;[1 1 1] Spawn lens effect and set duration parameter
              (if (ScalableFloat|GetValueasBool (Variables|Default|GetHF_EnableScreenEffect))
                (bind _lenseffectclass (Variables|Default|GetLensEffectClass)
                  (:then
                    (bind _returnvalue (Feedback|SpawnGenericCameraLensEffect _outfortplayerpawn _lenseffectclass))
                    (bind _output_get (Variables|Default|SetLensEffect _returnvalue))
                    (bind _asb_cle_iron_log_hit_player_center (Utilities|Casting|CastToB_CLE_IronLogHitPlayer_Center _output_get)
                      (:then
                        (Niagara|SetNiagaraVariable(Float) (Variables|/Script/Niagara.NiagaraActor|GetNiagaraComponent _asb_cle_iron_log_hit_player_center) (Variables|Default|GetFXDurationName) (ScalableFloat|GetValueatLevel (Variables|Default|GetFXDuration))))
                      (:CastFailed
                        _)))
                  (:else))))))
        (:then_1
          ;;[1 1 1] On pawn makeup effect
          (if (ScalableFloat|GetValueasBool (Variables|Default|GetHF_EnableOnPawnMakeupEffect))
            (Ability|AddGameplayCueWithParamsToOwner (Variables|Default|GetGameplay%20Cue%20Tag) (CallFunction|GetGameplayCueParameters))))))
    (:AvatarCastFailed
      _)))

(event Ability|EventOnEndAbility
  (params (bWasCancelled :type Boolean))
  ;[1 1 1] OnEnd
  (|CallParentFunction bWasCancelled)
  (Ability|HasAuthority
    (:False
      (bind _asb_cle_iron_log_hit_player_corners (Utilities|Casting|CastToB_CLE_IronLogHitPlayer_Corners (Variables|Default|GetLensEffect)))
      (Actor|DestroyActor _asb_cle_iron_log_hit_player_corners))))

(fn GetGameplayCueParameters
  (bind (_bblockinghit _binitialoverlap _time _distance _location _impactpoint _normal _impactnormal _physmat _hitactor _hitcomponent _hitbonename _bonename _hititem _elementindex _faceindex _tracestart _traceend) (Collision|BreakHitResult (Ability|EffectContext|GetHitResult (Ability|GetGrantedbyEffectContext))))
  (return (Ability|GameplayCue|MakeGameplayCueParameters :NormalizedMagnitude 0.0 :RawMagnitude 0.0 :Location _location :Normal (Math|Vector|Normalize (- _traceend _tracestart)) :GameplayEffectLevel 1 :AbilityLevel 1 :bReplicateLocationWhenUsingMinimalRepProxy true)))
