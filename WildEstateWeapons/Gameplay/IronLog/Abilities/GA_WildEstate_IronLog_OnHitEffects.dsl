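; Gameplay ability for on-hit effects of the WildEstate Iron Log weapon, currently a skeleton with no effect logic in this graph.
; On activation it resolves the avatar as a valid Fort player pawn and caches it in the FortPlayerPawn variable.
; Both the valid-pawn and cast-failed branches end the ability immediately.
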(event Ability|EventActivateAbility
  (bind _outfortplayerpawn (CallFunction|GetAvatarasValidFortPlayerPawn)
    (:ValidFortPlayerPawn
      (Variables|Default|SetFortPlayerPawn _outfortplayerpawn)
      (Ability|EndAbility))
    (:AvatarCastFailed
      (Ability|EndAbility))))
